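# Additive unlit glow material for Niagara mesh particles in the Grasshopper emote's rock effect.
# A fresnel-like rim term, built from the rock normal map and camera vector, hue-shifts particle color, which is scaled by particle alpha and passed through the FX emissive adjustments function.
# The graph also computes a vertex-normal Push offset and a rim mask weighted toward downward-facing normals, with no destination shown for either.
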Material /BRCosmetics/Athena/Effects/Particles/Emotes/NS_Emote_Grasshopper/M_Emote_Grasshopper_RockGlow
graph {
 properties {
  bUsedWithNiagaraMeshParticles=true
  blendMode=BLEND_Additive
  pixelDepthOffsetMode=PDOM_Legacy
  shadingModel=MSM_Unlit
  usedShadingModels=Unlit
 }
 nodes [
  n1 = ParticleColor()
  n2 = TextureSampleParameter2D() {
   parameterName=Normal Map
   samplerType=SAMPLERTYPE_Normal
   texture=/BRCosmetics/Animation/Props/Grasshopper_Rock/Textures/T_Emote_Prop_Grasshopper_Rock_N
  }
  n3 = DotProduct(A=@n2.RGB, B=Transform(None=CameraVectorWS()) {transformSourceType=TRANSFORMSOURCE_World, transformType=TRANSFORM_Tangent})
  n4 = Power(Base=OneMinus(None=Saturate(None=@n3)), Exp=ScalarParameter() {defaultValue=3, parameterName=Exponent})
  n5 = Multiply(A=@n4, B=ScalarParameter() {defaultValue=0.1, parameterName=Hue Shift})
  MaterialFunctionCall(Emissive=Multiply(
    A=@n1.A
    B=MaterialFunctionCall(Hue Shift Percentage=@n5, Texture=@n1.RGB) {materialFunction=&HueShift}.Result
   )) {materialFunction=&MF_FX_EmissiveAdjustments}
  n6 = VertexNormalWS()
  Multiply(A=@n6, B=ScalarParameter() {defaultValue=0.3, parameterName=Push})
  n7 = Constant3Vector() {constant=<rgba 0 0 -1 0>}
  Multiply(A=@n4, B=ConstantBiasScale(None=DotProduct(A=@n7, B=@n6)))
 ]
}
