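# Niagara system for the Grasshopper emote: a burst of persistent floating rock meshes with glow sprites, ringed around a central glow.
# FloatingRocks spawns on a tilted torus, with random mass driving size, glow and spin, then scales in.
# Springs pull rocks toward lagged, owner-transformed torus offsets while a sphere blocker, point force, curl noise and drag perturb them.
# MainRockGlow adds the looping center particle; both emitters follow owner scale and boost color in the front end.
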NiagaraSystem /BRCosmetics/Athena/Effects/Particles/Emotes/NS_Emote_Grasshopper/NS_Emote_Grasshopper
dependencies {usedDataInterfaces [&NiagaraDataInterfaceCurve], usedDynamicInputs [], usedModules [&SystemState]}
emitters {
 FloatingRocks {
  emitterUpdateScript [
   &EmitterState(
    Inactive Response=::NewEnumerator2
    Life Cycle Mode=::NewEnumerator1
    Loop Behavior=::NewEnumerator1
    Loop Duration Mode=::NewEnumerator1
    UseLoopDelay=false
   )
   &SpawnBurst_Instantaneous(Spawn Count=11, Use Spawn Probability=false, UseLoopCountLimit=false)
  ]
  particleSpawnScript [
   &InitializeParticle(
    Color=&MultiplyLinearColor(
     Linear Color=<rgba 0.0075 0.0245 0.05 1>
     Scale Factor=&MakeVector4_FromFloat(Float=&FN_FrontEndFloatModifier(Gameplay=1, Front End=2.222))
    )
    Lifetime=0.5
    Mass=&Lerp_Float(
     A=0.1
     B=2
     Alpha=&FloatFromCurve(
      FloatCurve=DataInterface() {propertyValues {Curve {keys [
          {interpMode=RCIM_Cubic, tangentMode=RCTM_User, time=0, value=0}
          {arriveTangent=3.2438, interpMode=RCIM_Cubic, leaveTangent=3.2438, tangentMode=RCTM_User, time=1, value=1}
         ]}}}
      CurveIndex=&RandomRangeFloat(Minimum=0, Maximum=1)
     )
    )
    Mass Mode=::NewEnumerator1
    UsePositionOffset=false
    Write Material Random=true
   )
   /Niagara/Modules/Update/Mass/CalculateSizeAndRotationalInertiaByMass(
    Calculate Mesh Scale=true
    Density by Material Type:ENiagaraMassByVolume=::NewEnumerator2
    Initial Model Dimensions:Vector3f=<100 100 100>
    "Manually Entered Density (kg/m^3)"=10
    Module.Manually Enter Density=false
   )
   &TorusLocation(
    Align Torus With Axis=true
    Coordinate Space:ENiagaraCoordinateSpace=::Simulation
    Handle Radius=33
    Large Radius=77
    MaskSpawnGroup=false
    Module.Override Randomness=false
    Torus Axis=&RandomRangeVector(Minimum=<0.1 -0.1 1>, Maximum=<-0.1 0.1 1>)
    Torus Origin:NiagaraPosition=/Niagara/DynamicInputs/Helpers/SimulationPosition()
   )
   /SetVariables(
    Particles.GlowSize=&Vector2DFromFloat(Value=&Multiply_Float(A=33, B=&Power(a=$Particles.Mass, b=0.333)))
    Particles.InitialTorusOffset:Vector3f=$Output.TorusLocation.TorusVector
    Particles.MeshIndex=&RandomRangeInt(Minimum=0, Maximum=2)
   ) {isSetParameters=true}
   &InitialMeshOrientation(Use Rotation Vector=false)
   &ApplyInitialForces()
  ]
  particleUpdateScript [
   &ParticleState(Kill Particles When Lifetime Has Elapsed=false)
   &InterpolateOverTime(
    Initial Vector Value:Vector3f=$Particles.Initial.InitialTorusOffset
    Rate Of Change=55
    Target Vector Value:Vector3f=&TransformVector(Vector=$Particles.Initial.InitialTorusOffset, Destination Space=::Simulation, Apply Scale=true)
    Variable Type=::NewEnumerator2
   )
   /SetVariables(Particles.InterpolatedTorusOffset:Vector3f=&Multiply_VectorByFloat(
     Vector=&NormalizeVector(
      Vector To Normalize=$StackContext.InterpolateOverTime.InterpolateOverTime_V3_.Moving Average
      Safe Normalize=true
      Safe Fallback Threshold=0
      Safe Fallback Vector=<0 0 0.001>
     )
     Float=&VectorLength(Vector=$Particles.Initial.InitialTorusOffset)
    )) {isSetParameters=true}
   &ScaleMeshSize(Scale Factor=&Multiply_VectorByFloat(
     Vector=$Engine.Owner.Scale
     Float=&FloatFromCurve(FloatCurve=DataInterface() {propertyValues {Curve {keys [{interpMode=RCIM_Cubic, tangentMode=RCTM_User, time=0.3, value=0}, {interpMode=RCIM_Cubic, tangentMode=RCTM_User, time=1, value=1}]}}})
    ))
   &self:SphereBlockingVolume(Sphere Origin:Vector3f=&SimulationPosition(), Sphere Radius=45) {module=SphereBlockingVolume}
   &PointForce(Force Falloff Distance=99, Force Strength=999, Module.Use Falloff=true, Module.UseLinearForce=false)
   &CurlNoiseForce(Noise Strength=11, Pan Noise=true, Pan Noise Field=<0 0 5>)
   &Drag(Drag=3, Use Linear Drag=true, Use Rotational Drag=false)
   /Niagara/Modules/Update/Forces/SpringForce(
    " Particle Equilibrium Position":NiagaraPosition=&Add_Vector(
     A=&SimulationPosition()
     B=&Multiply_Vector(A=$Engine.Owner.Scale, B=$StackContext.InterpolatedTorusOffset)
    )
    Dampening Coefficient=2
    Spring Tightness=22
   )
   &SolveForcesAndVelocity(Clamp Velocity=false, Limit Acceleration=false)
   &UpdateMeshOrientation(
    Rotation Rate=&Multiply_Float(
     A=&Divide_Float(A=1, B=$Particles.Mass)
     B=&RandomRangeFloat(Minimum=0.03, Maximum=0.1)
    )
    Rotation Vector=&RandomVector(Vector Scale=1)
   )
   /SetVariables(Particles.GlowSize=&Multiply_Vector2D(
     A=$Particles.Initial.GlowSize
     B=&Vector2DFromFloat(Value=&MakeFloatFromVector(
       Vector=&Multiply_VectorByFloat(
        Vector=$Engine.Owner.Scale
        Float=&FloatFromCurve(FloatCurve=DataInterface() {propertyValues {Curve {keys [{interpMode=RCIM_Cubic, tangentMode=RCTM_User, time=0.3, value=0}, {interpMode=RCIM_Cubic, tangentMode=RCTM_User, time=1, value=1}]}}})
       )
       Channel=::NewEnumerator0
      ))
    )) {isSetParameters=true}
   /SetVariables(
    Particles.DebugSpriteSize:Vector2f=<77 77>
    Particles.DebugValue:Vector4f=&MakeFloat4FromVectorAndFloat(
     "Vector (RGB)"=&TransformVector(Vector=$Particles.Initial.InitialTorusOffset, Destination Space=::Simulation)
     "Float (Alpha)"=1
    )
   ) {isSetParameters=true}
  ]
  rendererClasses [&NiagaraMeshRendererProperties, &NiagaraSpriteRendererProperties]
  renderers [
   {rendererClass=&NiagaraMeshRendererProperties, rendererIndex=0}
   {rendererClass=&NiagaraMeshRendererProperties, rendererIndex=1}
   {rendererClass=&NiagaraSpriteRendererProperties, rendererIndex=2}
  ]
 }
 MainRockGlow {
  emitterUpdateScript [
   &EmitterState(
    Inactive Response=::NewEnumerator2
    Life Cycle Mode=::NewEnumerator1
    Loop Behavior=::NewEnumerator1
    Loop Duration Mode=::NewEnumerator1
    UseLoopDelay=false
   )
   &SpawnBurst_Instantaneous(Use Spawn Probability=false, UseLoopCountLimit=false)
  ]
  particleSpawnScript [&InitializeParticle(
    Color=&MultiplyLinearColor(
     Linear Color=<rgba 0.03 0.098 0.2 1>
     Scale Factor=&MakeVector4_FromFloat(Float=&FN_FrontEndFloatModifier(Gameplay=1, Front End=2.222))
    )
    Lifetime=5
    Sprite Size Mode=::NewEnumerator3
    Uniform Sprite Size=150
    UsePositionOffset=false
    Write Material Random=true
   )]
  particleUpdateScript [
   &ParticleState(Kill Particles When Lifetime Has Elapsed=false, Loop Particles Lifetime=true)
   &CameraOffset()
   &ScaleSpriteSize(
    Scale Sprite Size Mode=::NewEnumerator0
    Uniform Scale Factor=&MakeFloatFromVector(Vector=$Engine.Owner.Scale, Channel=::NewEnumerator0)
   )
  ]
  rendererClasses [&NiagaraSpriteRendererProperties, &NiagaraMeshRendererProperties]
  renderers [
   {rendererClass=&NiagaraSpriteRendererProperties}
   {rendererClass=&NiagaraMeshRendererProperties, rendererIndex=1}
  ]
 }
}
summary {
 systemName=NS_Emote_Grasshopper
 emitters [
  {emitterName=FloatingRocks, rendererClasses [&NiagaraMeshRendererProperties, &NiagaraSpriteRendererProperties]}
  {emitterName=MainRockGlow, rendererClasses [&NiagaraSpriteRendererProperties, &NiagaraMeshRendererProperties]}
 ]
 userVariables [
  User.bIsFrontEnd {type=NiagaraBool}
  User.bIsFrontEndPreview {type=NiagaraBool}
  User.RibbonIDOffset {type=NiagaraInt32}
 ]
}
